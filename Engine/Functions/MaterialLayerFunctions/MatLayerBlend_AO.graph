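# Material layer blend function that applies an ambient occlusion map to a material's attributes to remove reflection.
# It uses Metallic to choose where the scalar AO lands: Base Color is darkened as the surface becomes metallic, Specular as it becomes dielectric.
# Two opposing lerps between AO and one supply these multipliers, and Metallic itself passes through unchanged.
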MaterialFunction /Engine/Functions/MaterialLayerFunctions/MatLayerBlend_AO
graph {
 properties {bExposeToLibrary=true, description=Blends an AO map to remove reflection., libraryCategoriesText [Misc, MaterialLayerBlend]}
 nodes [
  n1 = FunctionInput() {inputName=AO, inputType=FunctionInput_Scalar, sortPriority=1}
  n2 = GetMaterialAttributes(None=FunctionInput() {bUsePreviewValueAsDefault=true, inputName=Material, inputType=FunctionInput_MaterialAttributes}) {attributeGetTypes [#1, #2, #3]}
  n3 = LinearInterpolate(A=Constant() {r=1}, B=@n1, Alpha=ComponentMask(None=@n2.Metallic) {r=true})
  n4 = LinearInterpolate(A=@n1, B=Constant() {r=1}, Alpha=ComponentMask(None=@n2.Metallic) {r=true})
  FunctionOutput(None=SetMaterialAttributes!(
    MaterialAttributes=@n2.MaterialAttributes
    BaseColor=Multiply(A=@n3, B=@n2.BaseColor)
    Metallic=@n2.Metallic
    Specular=Multiply(A=@n4, B=@n2.Specular)
   ) {attributeSetTypes [#1, #2, #3]}) {outputName=Blended Material, sortPriority=-20}
 ]
}
