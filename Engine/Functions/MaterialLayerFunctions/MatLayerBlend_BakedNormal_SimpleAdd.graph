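# Material layer blend function, exposed to the library, that combines a high-poly baked normal map with a layered material's normal.
# It scales the incoming Material's normal Z down, adds the Normal input (flat by default), then scales the sum's Z down again.
# This simple additive blend skips renormalization and writes back only the Normal attribute, passing the other attributes through as Blended Material.
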MaterialFunction /Engine/Functions/MaterialLayerFunctions/MatLayerBlend_BakedNormal_SimpleAdd
graph {
 properties {bExposeToLibrary=true, description=Blends a Processed Normal Map with layered normals, libraryCategoriesText [Misc, MaterialLayerBlend]}
 nodes [
  n1 = GetMaterialAttributes(None=FunctionInput() {inputName=Material, inputType=FunctionInput_MaterialAttributes}) {attributeGetTypes [#1]}
  n2 = FunctionInput() {bUsePreviewValueAsDefault=true, desc=Normal map baked from high poly object., inputName=Normal, sortPriority=1, previewValue=<0 0 1 1>}
  n3 = Add(A=Multiply(A=@n1.Normal, B=Constant3Vector() {constant=<rgba 1 1 0.5 0>}), B=@n2)
  n4 = Constant3Vector() {constant=<rgba 1 1 0.675 0>}
  FunctionOutput(None=SetMaterialAttributes!(MaterialAttributes=@n1.MaterialAttributes, Normal=Multiply(A=@n3, B=@n4)) {attributeSetTypes [#1]}) {outputName=Blended Material, sortPriority=-20}
 ]
}
