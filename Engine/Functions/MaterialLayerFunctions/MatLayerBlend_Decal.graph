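# Material layer blend function that composites a decal sheet, sampled on the second UV channel, over a base material's attributes.
# The decal's alpha drives blends of Metallic toward zero and Roughness toward an input value, and, scaled by a flatness input, flattens the normal.
# Base color blends toward the decal color using its RGB, not alpha, as the mask.
# Opacity, SubsurfaceColor and CustomData0 are set to a default constant.
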MaterialFunction /Engine/Functions/MaterialLayerFunctions/MatLayerBlend_Decal
graph {
 properties {bExposeToLibrary=true, description=Blends in a decal sheet using the 2nd uv channel, libraryCategoriesText [Misc, MaterialLayerBlend]}
 nodes [
  n1 = GetMaterialAttributes(None=FunctionInput() {bUsePreviewValueAsDefault=true, inputName=Base Material, inputType=FunctionInput_MaterialAttributes}) {attributeGetTypes [
    #1
    #2
    #3
    #4
   ]}
  n2 = TextureSample(
   UVs=TextureCoordinate() {coordinateIndex=1}
   Tex=FunctionInput() {description=Color map with Alpha Channel, inputName=DecalSheet, inputType=FunctionInput_Texture2D}
  )
  n3 = FunctionInput() {bUsePreviewValueAsDefault=true, inputName=DecalRoughness, inputType=FunctionInput_Scalar, previewValue=<0.3 0 0 1>}
  n4 = Constant()
  n5 = FunctionInput() {bUsePreviewValueAsDefault=true, inputName=DecalNormalFlatness, inputType=FunctionInput_Scalar}
  n6 = SetMaterialAttributes!(
   MaterialAttributes=@n1.MaterialAttributes
   BaseColor=LinearInterpolate(A=@n1.BaseColor, B=@n2.RGB, Alpha=@n2.RGB)
   Metallic=LinearInterpolate(A=@n1.Metallic, B=Constant(), Alpha=@n2.A)
   Roughness=LinearInterpolate(A=@n1.Roughness, B=@n3, Alpha=@n2.A)
   Opacity=@n4
   Normal=LinearInterpolate(A=@n1.Normal, B=Constant3Vector() {constant=<rgba 0 0 1 0>}, Alpha=Multiply(A=@n2.A, B=@n5))
   SubsurfaceColor=@n4
   CustomData0=@n4
  ) {attributeSetTypes [
    #1
    #2
    #3
    #5
    #4
    #6
    #7
   ]}
  FunctionOutput(None=@n6) {outputName=Blended Material, sortPriority=-20}
 ]
}
