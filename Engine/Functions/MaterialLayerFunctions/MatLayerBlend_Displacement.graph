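# MatLayerBlend_Displacement is an engine material layer blend function, described as adding displacement to a material.
# It breaks the input Material's attributes and rebuilds them as the Blended Material output.
# As wired, the Displacement input feeds only an Add that reaches no output, and Tessellation Multiplier is unused.
# The rebuilt material's Displacement, Anisotropy, Tangent and ShadingModel pins are unconnected.
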MaterialFunction /Engine/Functions/MaterialLayerFunctions/MatLayerBlend_Displacement
graph {
 properties {description=Adds Displacement, libraryCategoriesText [Misc, MaterialLayerBlend]}
 nodes [
  Add(B=FunctionInput() {inputName=Displacement, sortPriority=2})
  FunctionInput() {inputName=Tessellation Multiplier, inputType=FunctionInput_Scalar, sortPriority=2}
  n1 = BreakMaterialAttributes!(Attr=FunctionInput() {inputName=Material, inputType=FunctionInput_MaterialAttributes})
  FunctionOutput(None=MakeMaterialAttributes!(
    BaseColor=@n1.BaseColor
    Metallic=@n1.BaseColor
    Specular=@n1.BaseColor
    Roughness=@n1.BaseColor
    Anisotropy
    EmissiveColor=@n1.BaseColor
    Opacity=@n1.BaseColor
    OpacityMask=@n1.BaseColor
    Normal=@n1.BaseColor
    Tangent
    WorldPositionOffset=@n1.BaseColor
    SubsurfaceColor=@n1.BaseColor
    ClearCoat=@n1.BaseColor
    ClearCoatRoughness=@n1.BaseColor
    AmbientOcclusion=@n1.BaseColor
    Refraction=@n1.BaseColor
    CustomizedUVs_0=@n1.BaseColor
    CustomizedUVs_1=@n1.BaseColor
    CustomizedUVs_2=@n1.BaseColor
    CustomizedUVs_3=@n1.BaseColor
    CustomizedUVs_4=@n1.BaseColor
    CustomizedUVs_5=@n1.BaseColor
    CustomizedUVs_6=@n1.BaseColor
    CustomizedUVs_7=@n1.BaseColor
    PixelDepthOffset=@n1.BaseColor
    ShadingModel
    Displacement
   )) {outputName=Blended Material, sortPriority=-20}
 ]
}
