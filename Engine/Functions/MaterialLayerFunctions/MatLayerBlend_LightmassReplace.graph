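# Material layer blend function that swaps a material's Base Color for a different color during Lightmass lighting builds.
# It reads Base Color from the Material input and feeds it into a LightmassReplace node as the realtime value, with LightmassReplaceColor, white by default, as the Lightmass value.
# The result is written back into Base Color, leaving other attributes unchanged in the Blended Material output.
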MaterialFunction /Engine/Functions/MaterialLayerFunctions/MatLayerBlend_LightmassReplace
graph {
 properties {bExposeToLibrary=true, description=Replace the Base Color in Lightmass, libraryCategoriesText [Misc, MaterialLayerBlend]}
 nodes [
  n1 = GetMaterialAttributes(None=FunctionInput() {bUsePreviewValueAsDefault=true, inputName=Material, inputType=FunctionInput_MaterialAttributes}) {attributeGetTypes [#1]}
  n2 = FunctionInput() {bUsePreviewValueAsDefault=true, inputName=LightmassReplaceColor, sortPriority=1, previewValue=<1 1 1 1>}
  FunctionOutput(None=SetMaterialAttributes!(MaterialAttributes=@n1.MaterialAttributes, BaseColor=LightmassReplace!(Realtime=@n1.BaseColor, Lightmass=@n2)) {attributeSetTypes [#1]}) {outputName=Blended Material, sortPriority=-20}
 ]
}
