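# MatLayerBlend_Multiply is a library material layer blend function that scales Material Attributes by a scalar Multiplier.
# It breaks the Base Material input, multiplies attributes from Metallic through ClearCoatRoughness, then rebuilds the output.
# BaseColor's multiply lacks the Multiplier connection, and a stray Multiply feeds nothing.
# AmbientOcclusion passes through and also feeds Refraction, CustomizedUVs and PixelDepthOffset, while Anisotropy, Tangent, ShadingModel and Displacement stay unconnected.
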MaterialFunction /Engine/Functions/MaterialLayerFunctions/MatLayerBlend_Multiply
graph {
 properties {
  bExposeToLibrary=true
  description=Multiplies all elements of Material Attributes by Multiplier
  libraryCategoriesText [Misc, MaterialLayerBlend]
 }
 nodes [
  n1 = BreakMaterialAttributes!(Attr=FunctionInput(Preview=MakeMaterialAttributes!(
     BaseColor
     Metallic
     Specular
     Roughness
     Anisotropy
     EmissiveColor
     Opacity
     OpacityMask
     Normal
     Tangent
     WorldPositionOffset
     SubsurfaceColor
     ClearCoat
     ClearCoatRoughness
     AmbientOcclusion
     Refraction
     CustomizedUVs_0
     CustomizedUVs_1
     CustomizedUVs_2
     CustomizedUVs_3
     CustomizedUVs_4
     CustomizedUVs_5
     CustomizedUVs_6
     CustomizedUVs_7
     PixelDepthOffset
     ShadingModel
     Displacement
    )) {bUsePreviewValueAsDefault=true, inputName=Base Material, inputType=FunctionInput_MaterialAttributes})
  n2 = FunctionInput() {bUsePreviewValueAsDefault=true, inputName=Multiplier, inputType=FunctionInput_Scalar, sortPriority=1, previewValue=<1 0 0 1>}
  FunctionOutput(None=MakeMaterialAttributes(
    BaseColor=Multiply(A=@n1.BaseColor)
    Metallic=Multiply(A=@n1.Metallic, B=@n2)
    Specular=Multiply(A=@n1.Specular, B=@n2)
    Roughness=Multiply(A=@n1.Roughness, B=@n2)
    EmissiveColor=Multiply(A=@n1.EmissiveColor, B=@n2)
    Opacity=Multiply(A=@n1.Opacity, B=@n2)
    OpacityMask=Multiply(A=@n1.OpacityMask, B=@n2)
    Normal=Multiply(A=@n1.Normal, B=@n2)
    WorldPositionOffset=Multiply(A=@n1.WorldPositionOffset, B=@n2)
    SubsurfaceColor=Multiply(A=@n1.SubsurfaceColor, B=@n2)
    ClearCoat=Multiply(A=@n1.ClearCoat, B=@n2)
    ClearCoatRoughness=Multiply(A=@n1.ClearCoatRoughness, B=@n2)
    AmbientOcclusion=@n1.AmbientOcclusion
    Refraction=@n1.AmbientOcclusion
    CustomizedUVs_0=@n1.AmbientOcclusion
    CustomizedUVs_1=@n1.AmbientOcclusion
    CustomizedUVs_2=@n1.AmbientOcclusion
    CustomizedUVs_3=@n1.AmbientOcclusion
    CustomizedUVs_4=@n1.AmbientOcclusion
    CustomizedUVs_5=@n1.AmbientOcclusion
    CustomizedUVs_6=@n1.AmbientOcclusion
    CustomizedUVs_7=@n1.AmbientOcclusion
    PixelDepthOffset=@n1.AmbientOcclusion
   )) {outputName=Blended Material, sortPriority=-20}
  Multiply(B=@n2)
 ]
}
