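# Material layer blend function, exposed to the library, described as replacing a material's WorldPositionDisplacement.
# It breaks the Material input and rebuilds it as Blended Material, leaving Anisotropy, Tangent, ShadingModel and Displacement unconnected.
# Two Mask-driven lerps take WorldPositionDisplacement and TessellationMultiplier as B, but neither has A connected or feeds the output, so the displacement override has no effect as wired.
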MaterialFunction /Engine/Functions/MaterialLayerFunctions/MatLayerBlend_OverrideDisplacement
graph {
 properties {bExposeToLibrary=true, description=Replaces WorldPositionDisplacement, libraryCategoriesText [Misc, MaterialLayerBlend]}
 nodes [
  n1 = BreakMaterialAttributes!(Attr=FunctionInput() {inputName=Material, inputType=FunctionInput_MaterialAttributes})
  FunctionOutput(None=MakeMaterialAttributes!(
    BaseColor=@n1.BaseColor
    Metallic=@n1.BaseColor
    Specular=@n1.BaseColor
    Roughness=@n1.BaseColor
    Anisotropy
    EmissiveColor=@n1.BaseColor
    Opacity=@n1.BaseColor
    OpacityMask=@n1.BaseColor
    Normal=@n1.BaseColor
    Tangent
    WorldPositionOffset=@n1.BaseColor
    SubsurfaceColor=@n1.BaseColor
    ClearCoat=@n1.BaseColor
    ClearCoatRoughness=@n1.BaseColor
    AmbientOcclusion=@n1.BaseColor
    Refraction=@n1.BaseColor
    CustomizedUVs_0=@n1.BaseColor
    CustomizedUVs_1=@n1.BaseColor
    CustomizedUVs_2=@n1.BaseColor
    CustomizedUVs_3=@n1.BaseColor
    CustomizedUVs_4=@n1.BaseColor
    CustomizedUVs_5=@n1.BaseColor
    CustomizedUVs_6=@n1.BaseColor
    CustomizedUVs_7=@n1.BaseColor
    PixelDepthOffset=@n1.BaseColor
    ShadingModel
    Displacement
   )) {outputName=Blended Material, sortPriority=-20}
  n2 = FunctionInput() {bUsePreviewValueAsDefault=true, inputName=Mask, inputType=FunctionInput_Scalar, sortPriority=2, previewValue=<1 0 0 1>}
  LinearInterpolate(B=FunctionInput() {inputName=WorldPositionDisplacement, sortPriority=1}, Alpha=@n2)
  n3 = FunctionInput() {bUsePreviewValueAsDefault=true, inputName=TessellationMultiplier, inputType=FunctionInput_Scalar, sortPriority=3, previewValue=<1 0 0 1>}
  LinearInterpolate(B=@n3, Alpha=@n2)
 ]
}
