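# Material layer blend function that replaces the Metallic value of an input material, exposed to the library under Misc and MaterialLayerBlend.
# It reads Metallic from the incoming attributes and linearly interpolates it toward NewMetalness using the scalar Mask, with preview values serving as input defaults.
# The result is written back into the attributes as Blended Material, along with the red channel of the original CustomData1.
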MaterialFunction /Engine/Functions/MaterialLayerFunctions/MatLayerBlend_OverrideMetalness
graph {
 properties {bExposeToLibrary=true, description=Replace the Metalness, libraryCategoriesText [Misc, MaterialLayerBlend]}
 nodes [
  n1 = GetMaterialAttributes(None=FunctionInput(Preview=MakeMaterialAttributes!(
     BaseColor
     Metallic
     Specular
     Roughness
     Anisotropy
     EmissiveColor
     Opacity
     OpacityMask
     Normal
     Tangent
     WorldPositionOffset
     SubsurfaceColor
     ClearCoat
     ClearCoatRoughness
     AmbientOcclusion
     Refraction
     CustomizedUVs_0
     CustomizedUVs_1
     CustomizedUVs_2
     CustomizedUVs_3
     CustomizedUVs_4
     CustomizedUVs_5
     CustomizedUVs_6
     CustomizedUVs_7
     PixelDepthOffset
     ShadingModel
     Displacement
    )) {bUsePreviewValueAsDefault=true, inputName=Material, inputType=FunctionInput_MaterialAttributes}) {attributeGetTypes [#1, #2]}
  n2 = FunctionInput() {bUsePreviewValueAsDefault=true, inputName=NewMetalness, sortPriority=1, previewValue=<1 1 1 1>}
  n3 = FunctionInput() {bUsePreviewValueAsDefault=true, inputName=Mask, inputType=FunctionInput_Scalar, sortPriority=2, previewValue=<1 1 1 1>}
  FunctionOutput(None=SetMaterialAttributes!(
    MaterialAttributes=@n1.MaterialAttributes
    Metallic=LinearInterpolate(A=@n1.Metallic, B=@n2, Alpha=@n3)
    CustomData1=ComponentMask(None=@n1.CustomData1) {r=true}
   ) {attributeSetTypes [#1, #2]}) {outputName=Blended Material, sortPriority=-20}
 ]
}
