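# Material layer blend function, exposed to the library, that blends a Top Material over a Base Material using a scalar Alpha.
# Alpha is clamped twice with separate min/max inputs: the normal range drives the Normal lerp, the color range drives the other lerped attributes.
# World Position Offset comes straight from the base, while attributes not explicitly set pass through from the top.
# A second output, Displacement, is left unconnected.
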MaterialFunction /Engine/Functions/MaterialLayerFunctions/MatLayerBlend_SeparateNormalandColorClamps
graph {
 properties {bExposeToLibrary=true, description=Blends all attributes of 2 Materials, libraryCategoriesText [Misc, MaterialLayerBlend]}
 nodes [
  n1 = GetMaterialAttributes(None=FunctionInput() {bUsePreviewValueAsDefault=true, inputName=Top Material, inputType=FunctionInput_MaterialAttributes, sortPriority=1}) {attributeGetTypes [
    #1
    #2
    #3
    #4
    #5
    #6
    #7
    #8
    #9
    #10
    #11
    #12
    #13
   ]}
  n2 = GetMaterialAttributes(None=FunctionInput() {bUsePreviewValueAsDefault=true, inputName=Base Material, inputType=FunctionInput_MaterialAttributes}) {attributeGetTypes [
    #1
    #2
    #3
    #4
    #5
    #6
    #7
    #8
    #9
    #10
    #11
    #12
    #13
   ]}
  n3 = FunctionInput() {bUsePreviewValueAsDefault=true, inputName=Alpha, inputType=FunctionInput_Scalar, sortPriority=2}
  n4 = FunctionInput() {bUsePreviewValueAsDefault=true, inputName=Color Blend Max, inputType=FunctionInput_Scalar, sortPriority=4, previewValue=<1 0 0 1>}
  n5 = Clamp(
   None=@n3
   Min=FunctionInput() {bUsePreviewValueAsDefault=true, inputName=Color Blend Min, inputType=FunctionInput_Scalar, sortPriority=3}
   Max=@n4
  )
  n6 = FunctionInput() {bUsePreviewValueAsDefault=true, inputName=Normal Blend Min, inputType=FunctionInput_Scalar, sortPriority=5}
  n7 = FunctionInput() {bUsePreviewValueAsDefault=true, inputName=Normal Blend Max, inputType=FunctionInput_Scalar, sortPriority=6, previewValue=<1 0 0 1>}
  n8 = SetMaterialAttributes!(
   MaterialAttributes=@n1.MaterialAttributes
   BaseColor=LinearInterpolate(A=@n2.BaseColor, B=@n1.BaseColor, Alpha=@n5)
   Metallic=LinearInterpolate(A=@n2.Metallic, B=@n1.Metallic, Alpha=@n5)
   Specular=LinearInterpolate(A=@n2.Specular, B=@n1.Specular, Alpha=@n5)
   Roughness=LinearInterpolate(A=@n2.Roughness, B=@n1.Roughness, Alpha=@n5)
   EmissiveColor=LinearInterpolate(A=@n2.EmissiveColor, B=@n1.EmissiveColor, Alpha=@n5)
   Opacity=LinearInterpolate(A=@n2.Opacity, B=@n1.Opacity, Alpha=@n5)
   OpacityMask=LinearInterpolate(A=@n2.OpacityMask, B=@n1.OpacityMask, Alpha=@n5)
   Normal=LinearInterpolate(A=@n2.Normal, B=@n1.Normal, Alpha=Clamp(None=@n3, Min=@n6, Max=@n7))
   WorldPositionOffset=@n2.WorldPositionOffset
   SubsurfaceColor=LinearInterpolate(A=@n2.SubsurfaceColor, B=@n1.SubsurfaceColor, Alpha=@n5)
   CustomData0=LinearInterpolate(A=@n2.CustomData0, B=@n1.CustomData0, Alpha=@n5)
   CustomData1=LinearInterpolate(A=@n2.CustomData1, B=@n1.CustomData1, Alpha=@n5)
   AmbientOcclusion=LinearInterpolate(A=@n2.AmbientOcclusion, B=@n1.AmbientOcclusion, Alpha=@n5)
  ) {attributeSetTypes [
    #1
    #2
    #3
    #4
    #5
    #6
    #7
    #8
    #9
    #10
    #11
    #12
    #13
   ]}
  FunctionOutput(None=@n8) {outputName=Blended Material, sortPriority=-10}
  FunctionOutput() {outputName=Displacement, sortPriority=-10}
 ]
}
