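# Material layer blend function that applies a top material as a stain over a base material, exposed to the material function library.
# Base color lerps by Alpha toward base color multiplied by the top color, which StainPremultiply first brightens.
# Roughness lerps by the same Alpha between base and top.
# No other top-layer attributes are used, and the result outputs as Blended Material.
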MaterialFunction /Engine/Functions/MaterialLayerFunctions/MatLayerBlend_Stain
graph {
 properties {
  bExposeToLibrary=true
  description="Blends the top layer as a stain, only the color and roughness from the top layer are used."
  libraryCategoriesText [Misc, MaterialLayerBlend]
 }
 nodes [
  n1 = GetMaterialAttributes(None=FunctionInput() {bUsePreviewValueAsDefault=true, inputName=Base Material, inputType=FunctionInput_MaterialAttributes}) {attributeGetTypes [#1, #2, #3]}
  n2 = GetMaterialAttributes(None=FunctionInput() {bUsePreviewValueAsDefault=true, inputName=Top Material, inputType=FunctionInput_MaterialAttributes, sortPriority=1}) {attributeGetTypes [#1, #3]}
  n3 = FunctionInput() {
   bUsePreviewValueAsDefault=true
   description=Brightens the stain before applying
   inputName=StainPremultiply
   inputType=FunctionInput_Scalar
   sortPriority=3
   previewValue=<2 0 0 1>
  }
  n4 = Multiply(A=@n1.BaseColor, B=Multiply(A=@n2.BaseColor, B=@n3))
  n5 = FunctionInput() {bUsePreviewValueAsDefault=true, inputName=Alpha, inputType=FunctionInput_Scalar, sortPriority=2}
  FunctionOutput(None=SetMaterialAttributes!(
    MaterialAttributes=@n1.MaterialAttributes
    BaseColor=LinearInterpolate(A=@n1.BaseColor, B=@n4, Alpha=@n5)
    Specular=@n1.MaterialAttributes
    Roughness=LinearInterpolate(A=@n1.Roughness, B=@n2.Roughness, Alpha=@n5)
   ) {attributeSetTypes [#1, #2, #3]}) {outputName=Blended Material, sortPriority=-20}
 ]
}
