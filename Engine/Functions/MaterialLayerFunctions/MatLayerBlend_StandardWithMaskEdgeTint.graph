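# Material layer blend function, exposed to the library, that blends a Base and Top material by a scalar Alpha.
# Twelve attributes, including normal, opacity and custom data, are linearly interpolated and written onto the base's attributes as Blended Material.
# A scalar Tint Mask multiplies only the top Base Color before blending; no edge is computed inside.
# A separate Displacement output comes from a lerp with only Alpha wired, so neither material feeds it.
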MaterialFunction /Engine/Functions/MaterialLayerFunctions/MatLayerBlend_StandardWithMaskEdgeTint
graph {
 properties {bExposeToLibrary=true, description=Blends all attributes of 2 Materials, libraryCategoriesText [Misc, MaterialLayerBlend]}
 nodes [
  n1 = GetMaterialAttributes(None=FunctionInput() {bUsePreviewValueAsDefault=true, inputName=Base Material, inputType=FunctionInput_MaterialAttributes}) {attributeGetTypes [
    #1
    #2
    #3
    #4
    #5
    #6
    #7
    #8
    #9
    #10
    #11
    #12
   ]}
  n2 = GetMaterialAttributes(None=FunctionInput() {bUsePreviewValueAsDefault=true, inputName=Top Material, inputType=FunctionInput_MaterialAttributes, sortPriority=1}) {attributeGetTypes [
    #1
    #2
    #3
    #4
    #5
    #6
    #7
    #8
    #9
    #10
    #11
    #12
   ]}
  n3 = FunctionInput() {bUsePreviewValueAsDefault=true, inputName=Tint Mask, inputType=FunctionInput_Scalar, sortPriority=3}
  n4 = FunctionInput() {bUsePreviewValueAsDefault=true, inputName=Alpha, inputType=FunctionInput_Scalar, sortPriority=2}
  n5 = SetMaterialAttributes!(
   MaterialAttributes=@n1.MaterialAttributes
   BaseColor=LinearInterpolate(A=@n1.BaseColor, B=Multiply(A=@n2.BaseColor, B=@n3), Alpha=@n4)
   Metallic=LinearInterpolate(A=@n1.Metallic, B=@n2.Metallic, Alpha=@n4)
   Specular=LinearInterpolate(A=@n1.Specular, B=@n2.Specular, Alpha=@n4)
   Roughness=LinearInterpolate(A=@n1.Roughness, B=@n2.Roughness, Alpha=@n4)
   EmissiveColor=LinearInterpolate(A=@n1.EmissiveColor, B=@n2.EmissiveColor, Alpha=@n4)
   Opacity=LinearInterpolate(A=@n1.Opacity, B=@n2.Opacity, Alpha=@n4)
   OpacityMask=LinearInterpolate(A=@n1.OpacityMask, B=@n2.OpacityMask, Alpha=@n4)
   Normal=LinearInterpolate(A=@n1.Normal, B=@n2.Normal, Alpha=@n4)
   SubsurfaceColor=LinearInterpolate(A=@n1.SubsurfaceColor, B=@n2.SubsurfaceColor, Alpha=@n4)
   CustomData0=LinearInterpolate(A=@n1.CustomData0, B=@n2.CustomData0, Alpha=@n4)
   CustomData1=LinearInterpolate(A=@n1.CustomData1, B=@n2.CustomData1, Alpha=@n4)
   AmbientOcclusion=LinearInterpolate(A=@n1.AmbientOcclusion, B=@n2.AmbientOcclusion, Alpha=@n4)
  ) {attributeSetTypes [
    #1
    #2
    #3
    #4
    #5
    #6
    #7
    #8
    #9
    #10
    #11
    #12
   ]}
  FunctionOutput(None=@n5) {outputName=Blended Material, sortPriority=-10}
  FunctionOutput(None=LinearInterpolate(Alpha=@n4)) {outputName=Displacement, sortPriority=1}
 ]
}
