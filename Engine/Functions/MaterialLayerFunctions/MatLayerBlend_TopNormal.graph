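# Material layer blend function, exposed to the library, that mixes Base and Top material attributes while taking the normal only from the Top material.
# Color, surface, opacity, subsurface, custom data, occlusion and refraction channels are lerped by the scalar Alpha, except CustomData0, whose lerp has no Alpha connected.
# Top's Normal passes straight through regardless of Alpha, and attributes not explicitly set carry over from Base.
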MaterialFunction /Engine/Functions/MaterialLayerFunctions/MatLayerBlend_TopNormal
graph {
 properties {
  bExposeToLibrary=true
  description=Blends Material Attributes but only uses the Top Normal Map
  libraryCategoriesText [Misc, MaterialLayerBlend]
 }
 nodes [
  n1 = GetMaterialAttributes(None=FunctionInput() {bUsePreviewValueAsDefault=true, inputName=Base Material, inputType=FunctionInput_MaterialAttributes}) {attributeGetTypes [
    #1
    #2
    #3
    #4
    #5
    #6
    #7
    #8
    #9
    #10
    #11
    #12
   ]}
  n2 = GetMaterialAttributes(None=FunctionInput() {bUsePreviewValueAsDefault=true, inputName=Top Material, inputType=FunctionInput_MaterialAttributes, sortPriority=1}) {attributeGetTypes [
    #1
    #2
    #3
    #4
    #5
    #6
    #7
    #8
    #9
    #10
    #11
    #12
    #13
   ]}
  n3 = FunctionInput() {bUsePreviewValueAsDefault=true, inputName=Alpha, inputType=FunctionInput_Scalar, sortPriority=2}
  n4 = SetMaterialAttributes!(
   MaterialAttributes=@n1.MaterialAttributes
   BaseColor=LinearInterpolate(A=@n1.BaseColor, B=@n2.BaseColor, Alpha=@n3)
   Metallic=LinearInterpolate(A=@n1.Metallic, B=@n2.Metallic, Alpha=@n3)
   Specular=LinearInterpolate(A=@n1.Specular, B=@n2.Specular, Alpha=@n3)
   Roughness=LinearInterpolate(A=@n1.Roughness, B=@n2.Roughness, Alpha=@n3)
   EmissiveColor=LinearInterpolate(A=@n1.EmissiveColor, B=@n2.EmissiveColor, Alpha=@n3)
   Opacity=LinearInterpolate(A=@n1.Opacity, B=@n2.Opacity, Alpha=@n3)
   OpacityMask=LinearInterpolate(A=@n1.OpacityMask, B=@n2.OpacityMask, Alpha=@n3)
   Normal=@n2.Normal
   SubsurfaceColor=LinearInterpolate(A=@n1.SubsurfaceColor, B=@n2.SubsurfaceColor, Alpha=@n3)
   CustomData0=LinearInterpolate(A=@n1.CustomData0, B=@n2.CustomData0)
   CustomData1=LinearInterpolate(A=@n1.CustomData1, B=@n2.CustomData1, Alpha=@n3)
   AmbientOcclusion=LinearInterpolate(A=@n1.AmbientOcclusion, B=@n2.AmbientOcclusion, Alpha=@n3)
   Refraction=LinearInterpolate(A=@n1.Refraction, B=@n2.Refraction, Alpha=@n3)
  ) {attributeSetTypes [
    #1
    #2
    #3
    #4
    #5
    #6
    #7
    #13
    #8
    #9
    #10
    #11
    #12
   ]}
  FunctionOutput(None=@n4) {outputName=Blended Material, sortPriority=-10}
 ]
}
