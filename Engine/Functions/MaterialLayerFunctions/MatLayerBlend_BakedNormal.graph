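# Material layer blend function that combines a baked normal map, typically from a high-poly mesh, with the normal of an input layered material.
# By default it uses BlendAngleCorrectedNormals with the baked map as base and the material normal as detail, then normalizes.
# A Quality Switch gives Low quality a cheaper path that adds only the material normal's XY to the baked normal, without normalizing.
# The result replaces only the Normal attribute, leaving other attributes untouched.
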MaterialFunction /Engine/Functions/MaterialLayerFunctions/MatLayerBlend_BakedNormal
graph {
 properties {bExposeToLibrary=true, description=Blends a Processed Normal Map with layered normals, libraryCategoriesText [Misc, MaterialLayerBlend]}
 nodes [
  n1 = GetMaterialAttributes(None=FunctionInput() {inputName=Material, inputType=FunctionInput_MaterialAttributes}) {attributeGetTypes [#1]}
  n2 = FunctionInput() {bUsePreviewValueAsDefault=true, desc=Normal map baked from high poly object., inputName=Normal, sortPriority=1, previewValue=<0 0 1 1>}
  n3 = Normalize(VectorInput=MaterialFunctionCall(BaseNormal=@n2, AdditionalNormal=@n1.Normal) {materialFunction=&BlendAngleCorrectedNormals}.Result)
  n4 = Add(A=@n2, B=Multiply(A=@n1.Normal, B=Constant3Vector() {constant=<rgba 1 1 0 0>}))
  FunctionOutput(None=SetMaterialAttributes!(MaterialAttributes=@n1.MaterialAttributes, Normal=QualitySwitch(Default=@n3, Low=@n4)) {attributeSetTypes [#1]}) {outputName=Blended Material, sortPriority=-20}
 ]
}
