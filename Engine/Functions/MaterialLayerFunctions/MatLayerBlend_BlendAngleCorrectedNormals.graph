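# Material layer blend function, exposed to the library, that layers an additional normal onto the normal of an incoming Material Attributes input.
# It combines the two normals with BlendAngleCorrectedNormals, then lerps from the original to that result by a scalar Mask, which defaults to full effect.
# Only the Normal attribute is replaced in the Blended Material output, although the description text says it replaces WorldPositionDisplacement.
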MaterialFunction /Engine/Functions/MaterialLayerFunctions/MatLayerBlend_BlendAngleCorrectedNormals
graph {
 properties {bExposeToLibrary=true, description=Replaces WorldPositionDisplacement, libraryCategoriesText [Misc, MaterialLayerBlend]}
 nodes [
  n1 = GetMaterialAttributes(None=FunctionInput() {inputName=Material, inputType=FunctionInput_MaterialAttributes}) {attributeGetTypes [#1]}
  n2 = FunctionInput() {bUsePreviewValueAsDefault=true, inputName=Additional Normal, sortPriority=1, previewValue=<0 0 1 1>}
  n3 = MaterialFunctionCall(BaseNormal=@n1.Normal, AdditionalNormal=@n2) {materialFunction=&BlendAngleCorrectedNormals}
  n4 = FunctionInput() {bUsePreviewValueAsDefault=true, inputName=Mask, inputType=FunctionInput_Scalar, sortPriority=2, previewValue=<1 0 0 1>}
  FunctionOutput(None=SetMaterialAttributes!(MaterialAttributes=@n1.MaterialAttributes, Normal=LinearInterpolate(A=@n1.Normal, B=@n3.Result, Alpha=@n4)) {attributeSetTypes [#1]}) {outputName=Blended Material, sortPriority=-20}
 ]
}
