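# MatLayerBlend_BreakBaseColor is an engine material function for material layer workflows that extracts Base Color from a Material Attributes input.
# The Material input feeds a GetMaterialAttributes node, whose BaseColor pin goes straight to the single BaseColor output.
# It is exposed to the library under MaterialLayerBlend; its description reads "Replace the Base Color", though the graph only reads that attribute.
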MaterialFunction /Engine/Functions/MaterialLayerFunctions/MatLayerBlend_BreakBaseColor
graph {
 properties {bExposeToLibrary=true, description=Replace the Base Color, libraryCategoriesText [Misc, MaterialLayerBlend]}
 nodes [FunctionOutput(None=GetMaterialAttributes(None=FunctionInput() {bUsePreviewValueAsDefault=true, inputName=Material, inputType=FunctionInput_MaterialAttributes}) {attributeGetTypes [#1]}.BaseColor) {outputName=BaseColor, sortPriority=-20}]
}
