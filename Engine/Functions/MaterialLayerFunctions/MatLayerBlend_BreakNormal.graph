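# MatLayerBlend_BreakNormal is a library material function that breaks the Normal out of a set of material attributes.
# A Material input, which falls back to its preview value by default, feeds a GetMaterialAttributes node that extracts only the Normal attribute.
# That value is returned through a single Normal output, and the function is exposed under the Misc and MaterialLayerBlend library categories.
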MaterialFunction /Engine/Functions/MaterialLayerFunctions/MatLayerBlend_BreakNormal
graph {
 properties {bExposeToLibrary=true, description=Break out the Normal, libraryCategoriesText [Misc, MaterialLayerBlend]}
 nodes [FunctionOutput(None=GetMaterialAttributes(None=FunctionInput() {bUsePreviewValueAsDefault=true, inputName=Material, inputType=FunctionInput_MaterialAttributes}) {attributeGetTypes [#1]}.Normal) {outputName=Normal, sortPriority=-20}]
}
