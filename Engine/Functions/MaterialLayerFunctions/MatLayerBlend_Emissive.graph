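# Material layer blend function, exposed to the library, that adds an emissive input to a material's existing emissive color.
# It reads EmissiveColor from the incoming Material Attributes, adds the Emissive input and clamps only the lower bound, keeping the sum non-negative.
# The result is written back into the attributes and returned as Blended Material, leaving all other attributes unchanged.
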MaterialFunction /Engine/Functions/MaterialLayerFunctions/MatLayerBlend_Emissive
graph {
 properties {bExposeToLibrary=true, description=Adds Emissive, libraryCategoriesText [Misc, MaterialLayerBlend]}
 nodes [
  n1 = GetMaterialAttributes(None=FunctionInput() {inputName=Material, inputType=FunctionInput_MaterialAttributes}) {attributeGetTypes [#1]}
  n2 = Add(A=@n1.EmissiveColor, B=FunctionInput() {inputName=Emissive, sortPriority=2})
  FunctionOutput(None=SetMaterialAttributes!(MaterialAttributes=@n1.MaterialAttributes, EmissiveColor=Clamp(None=@n2, Max=@n2) {clampMode=CMODE_ClampMin}) {attributeSetTypes [#1]}) {outputName=Blended Material, sortPriority=-20}
 ]
}
