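# Material layer blend function that modulates the Specular value of a base material, exposed to the library under Misc and MaterialLayerBlend.
# It reads Specular from the Base Material attributes, multiplies it by the scalar SpecularMultiplier input, and writes it back, passing all other attributes through as Blended Material.
# Both inputs use preview values as defaults, so an unconnected multiplier leaves Specular unchanged.
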MaterialFunction /Engine/Functions/MaterialLayerFunctions/MatLayerBlend_ModulateSpecular
graph {
 properties {bExposeToLibrary=true, description=Modulates the Specular value, libraryCategoriesText [Misc, MaterialLayerBlend]}
 nodes [
  n1 = GetMaterialAttributes(None=FunctionInput(Preview=MakeMaterialAttributes!(
     BaseColor
     Metallic
     Specular
     Roughness
     Anisotropy
     EmissiveColor
     Opacity
     OpacityMask
     Normal
     Tangent
     WorldPositionOffset
     SubsurfaceColor
     ClearCoat
     ClearCoatRoughness
     AmbientOcclusion
     Refraction
     CustomizedUVs_0
     CustomizedUVs_1
     CustomizedUVs_2
     CustomizedUVs_3
     CustomizedUVs_4
     CustomizedUVs_5
     CustomizedUVs_6
     CustomizedUVs_7
     PixelDepthOffset
     ShadingModel
     Displacement
    )) {bUsePreviewValueAsDefault=true, inputName=Base Material, inputType=FunctionInput_MaterialAttributes}) {attributeGetTypes [#1]}
  n2 = FunctionInput() {bUsePreviewValueAsDefault=true, inputName=SpecularMultiplier, inputType=FunctionInput_Scalar, sortPriority=1, previewValue=<1 0 0 1>}
  FunctionOutput(None=SetMaterialAttributes!(MaterialAttributes=@n1.MaterialAttributes, Specular=Multiply(A=@n1.Specular, B=@n2)) {attributeSetTypes [#1]}) {outputName=Blended Material, sortPriority=-20}
 ]
}
