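# Material layer blend function that adds a normal map, such as one baked from a high-poly object, on top of an input material under a mask.
# The mask lerps from a flat normal to the supplied Normal, and BlendAngleCorrectedNormals combines the result with the material's own normal.
# The normalized result replaces only the Normal attribute, leaving the other material attributes untouched.
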MaterialFunction /Engine/Functions/MaterialLayerFunctions/MatLayerBlend_NormalBlend
graph {
 properties {bExposeToLibrary=true, description=Blends a Normalmap on top of a material using a mask., libraryCategoriesText [Misc, MaterialLayerBlend]}
 nodes [
  n1 = GetMaterialAttributes(None=FunctionInput() {inputName=Material, inputType=FunctionInput_MaterialAttributes}) {attributeGetTypes [#1]}
  n2 = FunctionInput() {bUsePreviewValueAsDefault=true, desc=Normal map baked from high poly object., inputName=Normal, sortPriority=1, previewValue=<0 0 1 1>}
  n3 = FunctionInput() {bUsePreviewValueAsDefault=true, inputName=NormalMask, inputType=FunctionInput_Scalar, previewValue=<1 0 0 1>}
  n4 = MaterialFunctionCall(BaseNormal=@n1.Normal, AdditionalNormal=LinearInterpolate(A=Constant3Vector() {constant=<rgba 0 0 1 0>}, B=@n2, Alpha=@n3)) {materialFunction=&BlendAngleCorrectedNormals}
  FunctionOutput(None=SetMaterialAttributes!(MaterialAttributes=@n1.MaterialAttributes, Normal=Normalize(VectorInput=@n4.Result)) {attributeSetTypes [#1]}) {outputName=Blended Material, sortPriority=-20}
 ]
}
