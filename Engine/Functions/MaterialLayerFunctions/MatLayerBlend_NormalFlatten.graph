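# Material layer blend function that diminishes the effect of a material's normal map, exposed to the material function library.
# It extracts the Normal from the input Material attributes and lerps it toward a Normal input, which defaults to flat tangent-space, using the NormalFlatness scalar as alpha.
# The result replaces only the Normal attribute in the Blended Material output, leaving the other attributes untouched.
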MaterialFunction /Engine/Functions/MaterialLayerFunctions/MatLayerBlend_NormalFlatten
graph {
 properties {bExposeToLibrary=true, description=Diminishes the effect of a normalmap., libraryCategoriesText [Misc, MaterialLayerBlend]}
 nodes [
  n1 = GetMaterialAttributes(None=FunctionInput() {inputName=Material, inputType=FunctionInput_MaterialAttributes}) {attributeGetTypes [#1]}
  n2 = FunctionInput() {
   bUsePreviewValueAsDefault=true
   description="Defaults to a flat tangent space normal. (0,0,1)"
   inputName=Normal
   sortPriority=2
   previewValue=<0 0 1 1>
  }
  n3 = FunctionInput() {bUsePreviewValueAsDefault=true, inputName=NormalFlatness, inputType=FunctionInput_Scalar, sortPriority=1}
  FunctionOutput(None=SetMaterialAttributes!(MaterialAttributes=@n1.MaterialAttributes, Normal=LinearInterpolate(A=@n1.Normal, B=@n2, Alpha=@n3)) {attributeSetTypes [#1]}) {outputName=Blended Material, sortPriority=-20}
 ]
}
