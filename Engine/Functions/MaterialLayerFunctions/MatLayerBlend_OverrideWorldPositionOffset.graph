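# Material layer blend function that replaces the World Position Offset of an incoming material attributes set, exposed in the library under MaterialLayerBlend.
# It reads the input material's existing offset and lerps it toward the supplied WorldPositionOffset by a scalar Mask, which defaults to full override.
# The result is written back into the attributes and output as Blended Material, leaving all other attributes untouched.
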MaterialFunction /Engine/Functions/MaterialLayerFunctions/MatLayerBlend_OverrideWorldPositionOffset
graph {
 properties {bExposeToLibrary=true, description=Replaces WorldPositionOffset, libraryCategoriesText [Misc, MaterialLayerBlend]}
 nodes [
  n1 = GetMaterialAttributes(None=FunctionInput() {inputName=Material, inputType=FunctionInput_MaterialAttributes}) {attributeGetTypes [#1]}
  n2 = FunctionInput() {inputName=WorldPositionOffset, sortPriority=2}
  n3 = FunctionInput() {bUsePreviewValueAsDefault=true, inputName=Mask, inputType=FunctionInput_Scalar, sortPriority=2, previewValue=<1 0 0 1>}
  FunctionOutput(None=SetMaterialAttributes!(MaterialAttributes=@n1.MaterialAttributes, WorldPositionOffset=LinearInterpolate(A=@n1.WorldPositionOffset, B=@n2, Alpha=@n3)) {attributeSetTypes [#1]}) {outputName=Blended Material, sortPriority=-20}
 ]
}
