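# Material layer blend function, exposed to the material library, that replaces the normal of an incoming Material Attributes set.
# The Material input passes through Get and Set Material Attributes nodes, with only the Normal attribute overridden by the Normal input.
# The Normal input is optional and defaults to a flat tangent-space normal, so leaving it unconnected flattens the layer's normals.
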MaterialFunction /Engine/Functions/MaterialLayerFunctions/MatLayerBlend_ReplaceNormals
graph {
 properties {bExposeToLibrary=true, libraryCategoriesText [Misc, MaterialLayerBlend]}
 nodes [
  n1 = FunctionInput() {inputName=Material, inputType=FunctionInput_MaterialAttributes}
  n2 = FunctionInput() {
   bUsePreviewValueAsDefault=true
   description="Defaults to a flat tangent space normal. (0,0,1)"
   inputName=Normal
   sortPriority=2
   previewValue=<0 0 1 1>
  }
  FunctionOutput(None=SetMaterialAttributes!(MaterialAttributes=GetMaterialAttributes(None=@n1).MaterialAttributes, Normal=@n2) {attributeSetTypes [#1]})
 ]
}
