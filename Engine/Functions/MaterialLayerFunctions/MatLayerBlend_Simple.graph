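# MatLayerBlend_Simple is a material layer blend function, exposed to the library, that mixes two material attribute sets by a scalar Alpha.
# It linearly interpolates Base Color, Metallic, Roughness and Emissive Color from Base Material toward Top Material.
# Every other attribute, including Normal, passes through from Base Material, so the blend is narrower than its "all attributes except Normal" description.
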MaterialFunction /Engine/Functions/MaterialLayerFunctions/MatLayerBlend_Simple
graph {
 properties {bExposeToLibrary=true, description=Blends all attributes of 2 Materials except Normal., libraryCategoriesText [Misc, MaterialLayerBlend]}
 nodes [
  n1 = GetMaterialAttributes(None=FunctionInput() {bUsePreviewValueAsDefault=true, inputName=Base Material, inputType=FunctionInput_MaterialAttributes}) {attributeGetTypes [
    #1
    #2
    #3
    #4
   ]}
  n2 = GetMaterialAttributes(None=FunctionInput() {bUsePreviewValueAsDefault=true, inputName=Top Material, inputType=FunctionInput_MaterialAttributes, sortPriority=1}) {attributeGetTypes [
    #1
    #2
    #3
    #4
   ]}
  n3 = FunctionInput() {bUsePreviewValueAsDefault=true, inputName=Alpha, inputType=FunctionInput_Scalar, sortPriority=2}
  n4 = SetMaterialAttributes!(
   MaterialAttributes=@n1.MaterialAttributes
   BaseColor=LinearInterpolate(A=@n1.BaseColor, B=@n2.BaseColor, Alpha=@n3)
   Metallic=LinearInterpolate(A=@n1.Metallic, B=@n2.Metallic, Alpha=@n3)
   Roughness=LinearInterpolate(A=@n1.Roughness, B=@n2.Roughness, Alpha=@n3)
   EmissiveColor=LinearInterpolate(A=@n1.EmissiveColor, B=@n2.EmissiveColor, Alpha=@n3)
  ) {attributeSetTypes [
    #1
    #2
    #3
    #4
   ]}
  FunctionOutput(None=@n4) {outputName=Blended Material, sortPriority=-20}
 ]
}
