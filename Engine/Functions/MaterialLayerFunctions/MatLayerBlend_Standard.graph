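# Material layer blend function exposed to the library that blends all attributes of two materials.
# A BlendMaterialAttributes node mixes the Base Material and Top Material inputs by a scalar Alpha and outputs the Blended Material.
# Vertex attributes are taken from the Base Material only, while pixel attributes follow the Alpha blend.
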MaterialFunction /Engine/Functions/MaterialLayerFunctions/MatLayerBlend_Standard
graph {
 properties {bExposeToLibrary=true, description=Blends all attributes of 2 Materials, libraryCategoriesText [Misc, MaterialLayerBlend]}
 nodes [FunctionOutput(None=BlendMaterialAttributes(
    A=FunctionInput() {bUsePreviewValueAsDefault=true, inputName=Base Material, inputType=FunctionInput_MaterialAttributes}
    B=FunctionInput() {bUsePreviewValueAsDefault=true, inputName=Top Material, inputType=FunctionInput_MaterialAttributes, sortPriority=1}
    Alpha=FunctionInput() {bUsePreviewValueAsDefault=true, inputName=Alpha, inputType=FunctionInput_Scalar, sortPriority=2}
   ) {vertexAttributeBlendType=UseA}) {outputName=Blended Material, sortPriority=-10}]
}
